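# Physics asset for the MoonSatellite pickaxe, a two-handed melee weapon cosmetic, defining simplified collision for its skeletal mesh.
# A capsule covers the Melee_TwoHanded_Handle bone, and a larger, slightly tilted capsule covers b_AR_Root.
# A single constraint joins the two bodies with matching angular limits on both swing axes and twist.
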PhysicsAsset /BRCosmetics/Weapons/FORT_Melee/Pickaxe_MoonSatellite/Meshes/MoonSatellite_Axe_PhysicsAsset
bodies {
 Melee_TwoHanded_Handle [{shapeType=Capsule, length=33.865, radius=4.4457, center=<0 -0.0462 19.7815>, rotation=<rot -0.2498 -90 0>}]
 b_AR_Root [{shapeType=Capsule, length=52.6312, radius=8.0885, center=<2.4584 0 31.787>, rotation=<rot 0 -90 3.0939>}]
}
constraints [{bone1Name=Melee_TwoHanded_Handle, bone2Name=b_AR_Root, swing1LimitDegrees=45, swing2LimitDegrees=45, twistLimitDegrees=45}]
